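# Physics asset for the F_MED_Summer_Bag cosmetic mesh of a female medium player character, covering two dynamic anklet bones on the left leg.
# A capsule body on calf_l provides the anchor shape, and each anklet bone has its own thin box body.
# Each anklet is constrained directly to the calf, rather than chained to the other, with the same narrow swing and twist limit angles.
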PhysicsAsset /BRCosmetics/Characters/Player/Female/Medium/Bodies/F_MED_Summer/Meshes/F_MED_Summer_Bag_Physics
bodies {
 calf_l [{
   shapeType=Capsule
   length=8.231
   radius=5.6654
   shapeName=calf_l_capsule
   center=<-7.5777 -2.2764 -0.3167>
   rotation=<rot 0 -82.2363 -79.0874>
  }]
 dyn_anklet_01 [{
   shapeType=Box
   extentX=1.9775
   extentY=6.8891
   extentZ=6.9002
   shapeName=dyn_anklet_01_box
   center=<0.0557 0.029 0.0192>
   rotation=<rot 0 167.5116 -51.0998>
  }]
 dyn_anklet_02 [{
   shapeType=Box
   extentX=1.547
   extentY=6.5761
   extentZ=7.1516
   shapeName=dyn_anklet_02_box
   center=<-0.0764 0.0701 0.0022>
   rotation=<rot 0 -18.845 -62.4214>
  }]
}
constraints [
 {bone1Name=dyn_anklet_01, bone2Name=calf_l, swing1LimitDegrees=10, swing2LimitDegrees=10, twistLimitDegrees=10}
 {bone1Name=dyn_anklet_02, bone2Name=calf_l, swing1LimitDegrees=10, swing2LimitDegrees=10, twistLimitDegrees=10}
]
