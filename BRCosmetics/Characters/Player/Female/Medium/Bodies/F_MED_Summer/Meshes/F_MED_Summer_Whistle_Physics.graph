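# Physics asset for the whistle on the F_MED_Summer female medium player body, defining its dangle physics.
# A chain of capsule bodies runs from dyn_Whistle_base through three whistle bones, with angular limits widening toward the tip.
# The base joins spine_05 with free swing and twist.
# The spine_05 body combines a capsule and an angled box, apparently as a torso collision proxy.
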PhysicsAsset /BRCosmetics/Characters/Player/Female/Medium/Bodies/F_MED_Summer/Meshes/F_MED_Summer_Whistle_Physics
bodies {
 dyn_Whistle_01 [{shapeType=Capsule, length=8.812, radius=3.4591, shapeName=dyn_Whistle_01_capsule, center=<3.7004 -2.5 0>, rotation=<rot 0 -90 0>}]
 dyn_Whistle_02 [{shapeType=Capsule, length=7.3683, radius=2.6853, shapeName=dyn_Whistle_02_capsule, center=<3.4 -1.8 0>, rotation=<rot -3.5205 -90 0>}]
 dyn_Whistle_03 [{shapeType=Capsule, length=4.1329, radius=2.8765, shapeName=dyn_Whistle_03_capsule, center=<2.847 -1.3999 0>, rotation=<rot 0 -90 90>}]
 dyn_Whistle_base [{shapeType=Capsule, length=0.505, radius=0.505, shapeName=dyn_Whistle_base_capsule, rotation=<rot 0 -90 0>}]
 spine_05 [
  {shapeType=Capsule, length=10.2812, radius=9.3438, center=<-6.5498 4.2431 0>, rotation=<rot 0 0 -180>}
  {shapeType=Box, extentX=12.5, extentY=8.125, extentZ=23.25, center=<-5.4108 7.3613 0.0001>, rotation=<rot 0 -30.5471 0>}
 ]
}
constraints [
 {
  bone1Name=dyn_Whistle_base
  bone2Name=spine_05
  swing1LimitDegrees=45
  swing1Motion=Free
  swing2LimitDegrees=45
  swing2Motion=Free
  twistLimitDegrees=45
  twistMotion=Free
 }
 {bone1Name=dyn_Whistle_01, bone2Name=dyn_Whistle_base, swing1LimitDegrees=4, swing2LimitDegrees=4, twistLimitDegrees=4}
 {bone1Name=dyn_Whistle_02, bone2Name=dyn_Whistle_01, swing1LimitDegrees=12, swing2LimitDegrees=12, twistLimitDegrees=12}
 {bone1Name=dyn_Whistle_03, bone2Name=dyn_Whistle_02, swing1LimitDegrees=24, swing2LimitDegrees=24, twistLimitDegrees=24}
]
