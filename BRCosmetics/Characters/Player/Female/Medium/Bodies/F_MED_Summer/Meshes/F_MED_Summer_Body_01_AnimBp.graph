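# Animation Blueprint for the F_MED_Summer female body cosmetic, adding secondary motion to a pose copied from the attached parent mesh.
# The copied pose runs through the body's Control Rig, then a RigidBody whistle simulation in spine_04 space.
# Chained AnimDynamics nodes then spring the fanny pack, two anklets and both collar bones.
# Each anklet gets separate linear and angular passes, which, like the fanny pack, are LOD-limited; the event graph is empty.
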AnimBlueprint /BRCosmetics/Characters/Player/Female/Medium/Bodies/F_MED_Summer/Meshes/F_MED_Summer_Body_01_AnimBp
target_skeleton=&/F_MED_Summer_Body_01_Skeleton
graphs [
 AnimGraph {nodes [
   n1 = Control Rig(Source=Copy Pose From Mesh() {bUseAttachedParent=true}) {
    alphaBoolBlend {blendOption=Linear}
    controlRigAssetReference {blueprintRigClass=&/F_MED_Summer_Body_01_CtrlRig.F_MED_Summer_Body_01_CtrlRig_C}
   }
   n2 = RigidBody(ComponentPose=Local To Component(LocalPose=@n1)) {
    bOverrideWorldGravity=true
    overridePhysicsAsset=&/F_MED_Summer_Whistle_Physics
    simulationSpace=BaseBoneSpace
    alphaBoolBlend {blendOption=Linear}
    baseBoneRef {boneName=spine_04}
    componentAppliedLinearAccClamp=<0 0 0>
    overrideWorldGravity {z=-980}
    simSpaceSettings {
     dampingAlpha=0
     maxAngularAcceleration=30
     maxAngularVelocity=8
     maxLinearAcceleration=3500
     maxLinearVelocity=770
     velocityScaleZ=0.45
     worldAlpha=0.7
    }
   }
   n3 = Anim Dynamics - Bone: dyn_FannyPack(ComponentPose=@n2, Alpha=1, ExternalForce="0,0,0") {
    angularDampingOverride=0.97
    angularSpringConstant=8000
    bAngularSpring=true
    bGravityOverrideInSimSpace=true
    bOverrideAngularDamping=true
    bUseGravityOverride=true
    lODThreshold=1
    simulationSpace=BoneRelative
    alphaBoolBlend {blendOption=Linear}
    boundBone {boneName=dyn_FannyPack}
    componentAppliedLinearAccClamp=<770 770 770>
    componentLinearAccScale=<1 1 1>
    componentLinearVelScale=<1 1 1>
    gravityOverride {z=-980}
    physicsBodyDefinitions [{
      sphereCollisionRadius=0
      boundBone {boneName=dyn_FannyPack}
      boxExtents=<30 30 30>
      constraintSetup {angularTargetAxis=AxisY, twistAxis=AxisY, angularLimitsMax=<20 20 2>, angularLimitsMin=<-5 -20 -2>, angularTarget=<0 1 0>}
      localJointOffset=<0 4 0>
     }]
    relativeSpaceBone {boneName=root}
   }
   n4 = Anim Dynamics - Bone: dyn_anklet_01(ComponentPose=@n3, Alpha=1, ExternalForce="0,0,0") {
    angularDampingOverride=0.99
    angularSpringConstant=10000
    bGravityOverrideInSimSpace=true
    bLinearSpring=true
    bOverrideLinearDamping=true
    bUseGravityOverride=true
    lODThreshold=1
    linearDampingOverride=0.9
    linearSpringConstant=300
    simulationSpace=RootRelative
    alphaBoolBlend {blendOption=Linear}
    boundBone {boneName=dyn_anklet_01}
    componentAppliedLinearAccClamp=<1200 1200 1200>
    componentLinearAccScale=<1 1 1>
    componentLinearVelScale=<1 1 1>
    gravityOverride {z=-1200}
    physicsBodyDefinitions [{
      sphereCollisionRadius=0
      boundBone {boneName=dyn_anklet_01}
      boxExtents=<5 20 20>
      constraintSetup {angularLimitsMax=<0 0 0>, angularLimitsMin=<0 0 0>, angularTarget=<0 0 0>, linearAxesMax=<5 0 0>, linearAxesMin=<-0.5 0 0>}
      localJointOffset=<-3 0 0>
     }]
    relativeSpaceBone {boneName=calf_l}
   }
   n5 = Anim Dynamics - Bone: dyn_anklet_02(ComponentPose=@n4, Alpha=1, ExternalForce="0,0,0") {
    angularDampingOverride=0.99
    angularSpringConstant=10000
    bGravityOverrideInSimSpace=true
    bLinearSpring=true
    bOverrideLinearDamping=true
    bUseGravityOverride=true
    lODThreshold=1
    linearDampingOverride=0.9
    linearSpringConstant=300
    simulationSpace=RootRelative
    alphaBoolBlend {blendOption=Linear}
    boundBone {boneName=dyn_anklet_02}
    componentAppliedLinearAccClamp=<1200 1200 1200>
    componentLinearAccScale=<1 1 1>
    componentLinearVelScale=<1 1 1>
    gravityOverride {z=-1200}
    physicsBodyDefinitions [{
      sphereCollisionRadius=0
      boundBone {boneName=dyn_anklet_02}
      boxExtents=<5 20 20>
      constraintSetup {angularLimitsMax=<0 0 0>, angularLimitsMin=<0 0 0>, angularTarget=<0 0 0>, linearAxesMax=<3 0 0>, linearAxesMin=<-0.5 0 0>}
      localJointOffset=<-3 0 0>
     }]
    relativeSpaceBone {boneName=calf_l}
   }
   n6 = Anim Dynamics - Bone: dyn_anklet_01(ComponentPose=@n5, Alpha=1, ExternalForce="0,0,0") {
    angularDampingOverride=0.2
    angularSpringConstant=4000
    bAngularSpring=true
    bGravityOverrideInSimSpace=true
    bOverrideAngularBias=true
    bOverrideAngularDamping=true
    bUseGravityOverride=true
    bUsePlanarLimit=false
    lODThreshold=1
    linearDampingOverride=0.9
    linearSpringConstant=300
    simulationSpace=BoneRelative
    alphaBoolBlend {blendOption=Linear}
    boundBone {boneName=dyn_anklet_01}
    componentAppliedLinearAccClamp=<1200 1200 1200>
    componentLinearAccScale=<1 1 1>
    componentLinearVelScale=<1 1 1>
    gravityOverride {z=-1400}
    physicsBodyDefinitions [{
      sphereCollisionRadius=0
      boundBone {boneName=dyn_anklet_01}
      boxExtents=<10 15 15>
      constraintSetup {angularLimitsMax=<15 15 15>, angularLimitsMin=<-15 -15 -15>, angularTarget=<0 0 0>}
      localJointOffset=<-4 -4 2>
     }]
    relativeSpaceBone {boneName=calf_l}
   }
   n7 = Anim Dynamics - Bone: dyn_anklet_02(ComponentPose=@n6, Alpha=1, ExternalForce="0,0,0") {
    angularDampingOverride=0.2
    angularSpringConstant=4000
    bAngularSpring=true
    bGravityOverrideInSimSpace=true
    bOverrideAngularBias=true
    bOverrideAngularDamping=true
    bUseGravityOverride=true
    bUsePlanarLimit=false
    lODThreshold=1
    linearDampingOverride=0.9
    linearSpringConstant=300
    simulationSpace=BoneRelative
    alphaBoolBlend {blendOption=Linear}
    boundBone {boneName=dyn_anklet_02}
    componentAppliedLinearAccClamp=<1200 1200 1200>
    componentLinearAccScale=<1 1 1>
    componentLinearVelScale=<1 1 1>
    gravityOverride {z=-1400}
    physicsBodyDefinitions [{
      sphereCollisionRadius=0
      boundBone {boneName=dyn_anklet_02}
      boxExtents=<10 15 15>
      constraintSetup {angularLimitsMax=<15 15 15>, angularLimitsMin=<-15 -15 -15>, angularTarget=<0 0 0>}
      localJointOffset=<-3 2 -2>
     }]
    relativeSpaceBone {boneName=calf_l}
   }
   n8 = Anim Dynamics - Bone: dyn_collar_l(ComponentPose=@n7, Alpha=0.8, ExternalForce="0,0,0") {
    angularDampingOverride=1
    angularSpringConstant=20000
    bAngularSpring=true
    bOverrideAngularDamping=true
    bUseGravityOverride=true
    gravityScale=0.35
    simulationSpace=RootRelative
    alphaBoolBlend {blendOption=Linear}
    boundBone {boneName=dyn_collar_l}
    componentLinearAccScale=<1 1 1>
    componentLinearVelScale=<1 1 1>
    gravityOverride {z=-980}
    physicsBodyDefinitions [{
      boundBone {boneName=dyn_collar_l}
      boxExtents=<20 30 30>
      constraintSetup {angularLimitsMax=<15 15 0>, angularLimitsMin=<-15 -15 0>, angularTarget=<1 0 0>}
      localJointOffset=<4 0 0>
     }]
   }
   n9 = Anim Dynamics - Bone: dyn_collar_r(ComponentPose=@n8, Alpha=0.8, ExternalForce="0,0,0") {
    angularDampingOverride=1
    angularSpringConstant=20000
    bAngularSpring=true
    bOverrideAngularDamping=true
    bUseGravityOverride=true
    gravityScale=0.35
    simulationSpace=RootRelative
    alphaBoolBlend {blendOption=Linear}
    boundBone {boneName=dyn_collar_r}
    componentLinearAccScale=<1 1 1>
    componentLinearVelScale=<1 1 1>
    gravityOverride {z=-980}
    physicsBodyDefinitions [{
      boundBone {boneName=dyn_collar_r}
      boxExtents=<20 30 30>
      constraintSetup {angularLimitsMax=<15 15 0>, angularLimitsMin=<-15 -15 0>, angularTarget=<1 0 0>}
      localJointOffset=<-4 0 0>
     }]
   }
   Output Pose(Result=Component To Local(ComponentPose=@n9)) {layerGroup=None}
  ]}
 EventGraph {nodes [Event BlueprintUpdateAnimation()]}
]
variables [Pouch_Alpha {type=number}]
